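# Material function for the AgarBelly backpack that gives incoming material attributes a gummy, fake-translucent look.
# Sunlight reflected about the normal, compared with camera direction, produces forward and back scattering masks that tint a noise-varied, hue-shifted emissive color.
# A Fresnel plus local-position sphere-mask gradient is also built, and emissive passes through time-of-day adjustment.
# Base color is heavily dimmed, while tiled orange-peel detail, flattened by noise, supplies normal variation.
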MaterialFunction /BRCosmetics/Accessories/FORT_Backpacks/Backpack_F_MED_AgarBelly/Materials/MF_Gummy
graph {nodes [
  n1 = GetMaterialAttributes(None=FunctionInput() {inputType=FunctionInput_MaterialAttributes}) {attributeGetTypes [
    #1
    #2
    #3
    #4
    #5
   ]}
  n2 = TextureSampleParameter2D() {
   desc="R = SpecMask G = Metallic B = Roughness"
   group=Specular
   parameterName=SpecularMasks
   samplerType=SAMPLERTYPE_Masks
   texture=&T_Fortnite_Default_S
  }
  n3 = ScalarParameter() {defaultValue=20, group=MF_Gummy, parameterName=GummyRoughnessOffsetMax}
  n4 = ScalarParameter() {defaultValue=2.2, group=MF_Gummy, parameterName=GummyRoughnessOffsetMin}
  n5 = NamedRerouteUsage()
  n6 = NamedRerouteUsage()
  n7 = NamedRerouteUsage()
  n8 = SetMaterialAttributes!(
   MaterialAttributes=@n1.MaterialAttributes
   BaseColor=Multiply(A=NamedRerouteUsage(), B=ScalarParameter() {defaultValue=0.05, group=MF_Gummy, parameterName=Gummy_BaseColorBlend})
   EmissiveColor=MaterialFunctionCall!(Color=Multiply(A=@n6, B=@n7)) {materialFunction=/Game/Packages/Fortress_SharedMaterials/Anime_Master_Material/MF_TODColorAdjustment}.Result
   Metallic=@n2.G
   Specular=@n2.G
   Roughness=Multiply(A=@n2.B, B=LinearInterpolate(A=@n3, B=@n4, Alpha=@n5))
   Normal=NamedRerouteUsage()
   Opacity
   SubsurfaceColor
  ) {attributeSetTypes [
    #1
    #6
    #2
    #3
    #4
    #5
    #7
    #8
   ]}
  FunctionOutput(None=@n8)
  n9 = CollectionParameter() {
   collection=&FortniteMaterialParameters
   parameterName=SunlightDirectionalVector
  }
  n10 = ComponentMask(None=@n9) {b=true, g=true, r=true}
  n11 = StaticSwitch(False=PixelNormalWS(), Value=StaticBoolParameter() {group=Slurp, parameterName=UseSlurpNormalDistortion})
  n12 = Multiply(A=Multiply(A=DotProduct(A=@n10, B=@n11)) {constB=2}, B=@n11)
  n13 = DotProduct(
   A=Normalize(VectorInput=Subtract(A=@n12, B=@n10))
   B=Multiply(A=MaterialFunctionCall() {materialFunction=&CameraDirectionVector}.Result) {constB=-1}
  )
  n14 = Power(Base=Max(A=@n13) {constB=0}, Exp=ScalarParameter() {defaultValue=1, group=Slurp, parameterName=Slurp_InnerScatteringPower})
  n15 = ScalarParameter() {defaultValue=0.4, group=Slurp, parameterName=Slurp_InnerScatteringScale}
  NamedRerouteDeclaration(None=Saturate(None=Multiply(A=@n14, B=@n15) {constB=0.4})) {name=Scattering Mask, nodeColor=<rgba 0 0.1797 1 1>}
  NamedRerouteDeclaration(None=@n1.Normal) {name=In_Normal, nodeColor=<rgba 0.8281 1 0 1>}
  n16 = TextureSample(UVs=Multiply(A=TextureCoordinate(), B=ScalarParameter() {defaultValue=7, group=MF_Gummy, parameterName=DetailNormalTiling})) {samplerType=SAMPLERTYPE_Normal, texture=/Game/Global/Textures/Default/Noise/T_OrangePeel_N}
  NamedRerouteDeclaration(None=@n16.B) {name=NormalNoise, nodeColor=<rgba 0.5 0 1 1>}
  n17 = TextureSample(UVs=Multiply(A=TextureCoordinate(), B=Constant() {r=10})) {samplerSource=SSM_Wrap_WorldGroupSettings, samplerType=SAMPLERTYPE_Grayscale, texture=&LowResBlurredNoise}
  NamedRerouteDeclaration(None=MaterialFunctionCall(
    BaseNormal=NamedRerouteUsage()
    AdditionalNormal=MaterialFunctionCall(Normal=@n16.RGB, Flatness=@n17.R) {materialFunction=&FlattenNormal}.Result
   ) {materialFunction=&BlendAngleCorrectedNormals}.Result) {name=Out_Normal, nodeColor=<rgba 0.8516 0 1 1>}
  n18 = Desaturation(None=Saturate(None=NamedRerouteUsage()), Fraction=ScalarParameter() {defaultValue=-0.2, group=MF_Gummy, parameterName=Gummy_HueShiftAmount})
  n19 = Saturate(None=NamedRerouteUsage())
  n20 = NamedRerouteUsage()
  n21 = TextureSampleParameter2D() {
   desc="Masks R=AO G= Fresnel B = Tmission/Skin"
   group=Masks
   parameterName=M
   samplerType=SAMPLERTYPE_Masks
   texture=&T_BlackMask
  }
  n22 = ScalarParameter() {defaultValue=0.5, group=MF_Gummy, parameterName=Gummy_AOPower}
  n23 = LinearInterpolate(
   A=Constant()
   B=Multiply(
    A=LinearInterpolate(A=@n18, B=@n19, Alpha=@n20)
    B=ScalarParameter() {defaultValue=0.85, group=MF_Gummy, parameterName=Gummy_ColorBrightness}
   )
   Alpha=Saturate(None=Power(Base=@n21.G, Exp=@n22))
  )
  n24 = VectorParameter() {group=MF_Gummy, parameterName=Slurp_InnerScatteringColor, defaultValue=<rgba 1 0 0.016 1>}
  n25 = NamedRerouteUsage()
  n26 = Constant3Vector() {constant=<rgba 1 1 1 0>}
  n27 = VectorParameter() {group=MF_Gummy, parameterName=Slurp_InnerBackScatteringMultiplier, defaultValue=<rgba 0 0 0 1>}
  n28 = NamedRerouteUsage()
  NamedRerouteDeclaration(None=Multiply(A=LinearInterpolate(A=@n23, B=@n24.RGB, Alpha=@n25), B=LinearInterpolate(A=@n26, B=@n27.RGB, Alpha=@n28))) {name=OutEmissive, nodeColor=<rgba 0 1 0.1797 1>}
  NamedRerouteDeclaration(None=@n1.BaseColor) {name=In_BaseColor, nodeColor=<rgba 0.0547 1 0 1>}
  n29 = Multiply(A=@n13) {constB=-1}
  n30 = ScalarParameter() {group=Slurp, parameterName=Slurp_InnerBackScatteringBias}
  n31 = Power(
   Base=Max(A=Add(A=@n29, B=@n30)) {constB=0}
   Exp=ScalarParameter() {defaultValue=1, group=Slurp, parameterName=Slurp_InnerBackScatteringPower}
  )
  n32 = ScalarParameter() {defaultValue=0.4, group=Slurp, parameterName=Slurp_InnerBackScatteringScale}
  NamedRerouteDeclaration(None=Saturate(None=Multiply(A=@n31, B=@n32) {constB=0.55})) {name=BackScatteringMask, nodeColor=<rgba 0 0.7656 1 1>}
  NamedRerouteDeclaration(None=Multiply(A=NamedRerouteUsage(), B=ScalarParameter() {defaultValue=1, group=MF_Gummy, parameterName=Gummy_ColorADarkOffset})) {name=Gummy_ColorDark, nodeColor=<rgba 0.8984 0 1 1>}
  n33 = TextureSample(UVs=Multiply(A=TextureCoordinate(), B=ScalarParameter() {defaultValue=14.3676, group=MF_Gummy, parameterName=Gummy_TintNoiseTiling})) {samplerSource=SSM_Wrap_WorldGroupSettings, samplerType=SAMPLERTYPE_Grayscale, texture=&LowResBlurredNoise}
  NamedRerouteDeclaration(None=Saturate(None=@n33.R)) {name=Tint Mask_1, nodeColor=<rgba 0.6406 0 1 1>}
  n34 = Fresnel(ExponentIn=ScalarParameter() {defaultValue=2, group=MF_Gummy, parameterName=Gummy_Fresnel_Power})
  n35 = ScalarParameter() {defaultValue=1, group=MF_Gummy, parameterName=Gummy_Fresnel_Amount}
  n36 = ScalarParameter() {defaultValue=0.5, group=MF_Gummy, parameterName=Gummy_HeightGradient_Min}
  n37 = ScalarParameter() {defaultValue=1.6, group=MF_Gummy, parameterName=Gummy_HeightGradient_Max}
  n38 = VertexInterpolator(VS=LocalPosition() {localOrigin=InstancePreSkinning}.XYZ)
  n39 = VectorParameter() {group=MF_Gummy, parameterName=Gummy_LoccalPosition, defaultValue=<rgba 0 -12.2677 8.9353 1>}
  n40 = ScalarParameter() {defaultValue=21, group=MF_Gummy, parameterName=Gummy_HeightRadius}
  n41 = ScalarParameter() {group=MF_Gummy, parameterName=Gummy_HeightHardness}
  n42 = Saturate(None=ComponentMask(None=SphereMask(A=@n38.PS, B=@n39.RGB, Radius=@n40, Hardness=@n41)) {r=true})
  NamedRerouteDeclaration(None=Add(A=Multiply(A=@n34, B=@n35) {constB=0.5}, B=LinearInterpolate(A=@n36, B=@n37, Alpha=@n42))) {name=Gummy_HeightGradient, nodeColor=<rgba 0.1719 0 1 1>}
 ]}
