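# Parent material for the AgarBelly gummy backpack cosmetic, layering region-specific emissive color onto MF_Gummy between character PreFX, skin FX and PostFX functions.
# A packed mask texture supplies hat, body and emissive-noise masks.
# Emissive blends combine hat edge tint, body color with Fresnel glow, and noise brightness.
# Sphere masks on X-scaled pre-skinned local position shape hat glow and a head mask, so both follow the deforming mesh.
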Material /BRCosmetics/Accessories/FORT_Backpacks/Backpack_F_MED_AgarBelly/Materials/M_Gummy_Parent
graph {
 properties {
  bUseMaterialAttributes=true
  bUsedWithMeshDeformer=true
  bUsedWithMorphTargets=true
  bUsedWithSkeletalMesh=true
  pixelDepthOffsetMode=PDOM_Legacy
  refractionMethod=RM_IndexOfRefraction
  usedShadingModels=Default Lit
 }
 nodes [
  n1 = MaterialFunctionCall() {materialFunction=&BaseFN_CharacterFunction_PreFX}
  n2 = GetMaterialAttributes(None=MaterialFunctionCall!(In=MaterialFunctionCall(
     PreskinnedPosition=@n1.Pre skinned Local Position
     Gmap=@n1.Pre skinned Local Position
     MaterialAttributes=@n1.Pre skinned Local Position
     AO=@n1.Pre skinned Local Position
    ) {materialFunction=&CharacterSkinFX}.Result) {materialFunction=&/MF_Gummy}.Result) {attributeGetTypes [#1]}
  n3 = VectorParameter() {group=AgarBelly, parameterName=HatEdgeColor, defaultValue=<rgba 0.08 0.024 0.0286 1>}
  n4 = ScalarParameter() {defaultValue=0.5, parameterName=HatEdgeColorBlend}
  n5 = Add(
   A=LinearInterpolate(A=@n2.Emissive Color, B=LinearInterpolate(A=@n2.Emissive Color, B=@n3.RGB, Alpha=@n4), Alpha=NamedRerouteUsage())
   B=NamedRerouteUsage()
  )
  n6 = LinearInterpolate(
   A=@n5
   B=VectorParameter() {group=AgarBelly, parameterName=BodyEmissiveColor, defaultValue=<rgba 0.0666 0.2961 0.0529 1>}.RGB
   Alpha=ScalarParameter() {defaultValue=0.22, group=AgarBelly, parameterName=BodyEmissiveColorBlend}
  )
  n7 = ScalarParameter() {defaultValue=2, group=AgarBelly, parameterName=BodyEmissiveFresnel}
  n8 = Multiply(
   A=VectorParameter() {group=AgarBelly, parameterName=BodyEmissiveColor2, defaultValue=<rgba 0.0375 0.75 0.0419 1>}.RGB
   B=Multiply(A=Fresnel(ExponentIn=@n7), B=ScalarParameter() {defaultValue=0.1, group=AgarBelly, parameterName=BodyEmissive_Intensity})
  )
  n9 = Multiply(
   A=LinearInterpolate(A=@n5, B=Add(A=@n6, B=@n8), Alpha=NamedRerouteUsage())
   B=LinearInterpolate(B=ScalarParameter() {defaultValue=2.5, group=AgarBelly, parameterName=EmissiveNoiseColorBrightness}, Alpha=NamedRerouteUsage()) {constA=1.1}
  )
  n10 = LinearInterpolate(
   A=@n9
   B=Constant4Vector() {constant=<rgba 0.4609 0.0605 0 1>}.RGB
   Alpha=ScalarParameter() {defaultValue=1, parameterName=HeadEmissiveBlend}
  )
  n11 = NamedRerouteUsage()
  MaterialFunctionCall(
   AO=@n1.AO
   Distance Blend=@n1.AO
   Fresnel=@n1.AO
   Gmap=@n1.AO
   Pre Skinned Local Position=@n1.AO
   Material=SetMaterialAttributes(MaterialAttributes=@n2.MaterialAttributes, Emissive Color=LinearInterpolate(A=@n9, B=@n10, Alpha=@n11)) {attributeSetTypes [#1]}
  ) {materialFunction=&BaseFN_CharacterFunction_PostFX}
  n12 = LocalPosition() {localOrigin=InstancePreSkinning}
  n13 = MaterialFunctionCall(
   R=Multiply(A=ComponentMask(None=@n12.XYZ) {r=true}, B=Constant() {r=1.3413})
   G=ComponentMask(None=@n12.XYZ) {g=true}
   B=ComponentMask(None=@n12.XYZ) {b=true}
  ) {materialFunction=&AppendMany}
  n14 = OneMinus(None=SphereMask(
    A=VertexInterpolator(VS=@n13.RGB).PS
    B=Constant4Vector() {constant=<rgba 0 -9.9011 10.0231 1>}.RGB
    Radius=Constant() {r=18.0831}
    Hardness=Constant() {r=0.2949}
   ))
  n15 = NamedRerouteUsage()
  NamedRerouteDeclaration(None=Multiply(
    A=LinearInterpolate(B=@n14, Alpha=@n15)
    B=VectorParameter() {group=AgarBelly, parameterName=HatEmissiveColor, defaultValue=<rgba 0.3 0.0765 0.0391 1>}.RGB
   )) {name=HatEmissive, nodeColor=<rgba 1 0 0.375 1>}
  n16 = LocalPosition() {localOrigin=InstancePreSkinning}
  n17 = ComponentMask(None=@n16.XYZ) {r=true}
  n18 = Constant() {r=1.3413}
  n19 = VertexInterpolator(VS=MaterialFunctionCall(R=Multiply(A=@n17, B=@n18), G=ComponentMask(None=@n16.XYZ) {g=true}, B=ComponentMask(None=@n16.XYZ) {b=true}) {materialFunction=&AppendMany}.RGB)
  n20 = Constant4Vector() {constant=<rgba 0 -9.9011 10.0231 1>}
  n21 = Constant() {r=7.2873}
  n22 = Constant() {r=0.1909}
  NamedRerouteDeclaration(None=LinearInterpolate(A=SphereMask(A=@n19.PS, B=@n20.RGB, Radius=@n21, Hardness=@n22), Alpha=NamedRerouteUsage()) {constB=0}) {name=Head, nodeColor=<rgba 1 0 0.3047 1>}
  n23 = TextureSample() {samplerType=SAMPLERTYPE_Masks, texture=/BRCosmetics/Accessories/FORT_Backpacks/Backpack_F_MED_AgarBelly/Textures/T_AgarBelly_Backpack_TAMask}
  NamedRerouteDeclaration(None=@n23.R) {name=HatMask, nodeColor=<rgba 1 0 0.8672 1>}
  NamedRerouteDeclaration(None=@n23.G) {name=BodyMask, nodeColor=<rgba 0 1 0.8125 1>}
  NamedRerouteDeclaration(None=@n23.B) {name=EmissiveNoiseMask, nodeColor=<rgba 1 0.4922 0 1>}
 ]
}
