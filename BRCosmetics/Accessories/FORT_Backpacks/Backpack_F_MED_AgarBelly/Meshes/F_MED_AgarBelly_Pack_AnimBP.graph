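# Animation Blueprint for the AgarBelly backpack cosmetic, adding spring-driven secondary motion to its arm, leg and head bones.
# The cached input pose feeds two chains of RootRelative AnimDynamics nodes, selected by bIsInFrontEnd.
# The front-end chain first adds small rotation offsets to three limb bones and uses retuned, mostly wider limb limits.
# The other chain simulates directly on the input pose, and the EventGraph has no logic.
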AnimBlueprint /BRCosmetics/Accessories/FORT_Backpacks/Backpack_F_MED_AgarBelly/Meshes/F_MED_AgarBelly_Pack_AnimBP
target_skeleton=&/F_MED_AgarBelly_Pack_Skeleton
graphs [
 AnimGraph {nodes [
   n1 = Anim Dynamics - Bone: dyn_arm_r(ComponentPose=Local To Component(LocalPose=Use cached pose 'Input saved pose'()), Alpha=1, ExternalForce="0,0,0") {
    angularDampingOverride=1
    angularSpringConstant=18000
    bAngularSpring=true
    bOverrideAngularDamping=true
    bUseGravityOverride=true
    gravityScale=0.35
    simulationSpace=RootRelative
    boundBone {boneName=dyn_arm_r}
    componentAppliedLinearAccClamp=<2500 2500 2500>
    componentLinearAccScale=<0.6 0.6 0.6>
    componentLinearVelScale=<0.6 0.6 0.6>
    gravityOverride {z=-980}
    physicsBodyDefinitions [{
      boundBone {boneName=dyn_arm_r}
      boxExtents=<29 29 29>
      constraintSetup {angularLimitsMax=<2 13 17>, angularLimitsMin=<-2 -13 -19>, angularTarget=<1 0 0>}
      localJointOffset=<3 0 0>
     }]
    simSpaceSettings {maxAngularAcceleration=60, maxAngularVelocity=6, simSpaceAngularAlpha=0.8}
   }
   n2 = Anim Dynamics - Bone: dyn_arm_l(ComponentPose=@n1, Alpha=1, ExternalForce="0,0,0") {
    angularDampingOverride=1
    angularSpringConstant=17000
    bAngularSpring=true
    bOverrideAngularDamping=true
    bUseGravityOverride=true
    gravityScale=0.35
    simulationSpace=RootRelative
    boundBone {boneName=dyn_arm_l}
    componentAppliedLinearAccClamp=<3000 3000 3000>
    componentLinearAccScale=<0.3 0.3 0.3>
    componentLinearVelScale=<0.3 0.3 0.3>
    gravityOverride {z=-980}
    physicsBodyDefinitions [{
      boundBone {boneName=dyn_arm_l}
      boxExtents=<20 30 30>
      constraintSetup {angularLimitsMax=<0 7 16>, angularLimitsMin=<0 -7 -13>, angularTarget=<1 0 0>}
      localJointOffset=<2.5 0 0>
     }]
    simSpaceSettings {maxAngularAcceleration=65, maxAngularVelocity=6.5, simSpaceAngularAlpha=0.8}
   }
   n3 = Anim Dynamics - Bone: dyn_leg_l(ComponentPose=@n2, Alpha=1, ExternalForce="0,0,0") {
    angularDampingOverride=1
    angularSpringConstant=17000
    bAngularSpring=true
    bOverrideAngularDamping=true
    bUseGravityOverride=true
    gravityScale=0.35
    simulationSpace=RootRelative
    boundBone {boneName=dyn_leg_l}
    componentAppliedLinearAccClamp=<2500 2500 2500>
    componentLinearAccScale=<0.4 0.4 0.4>
    componentLinearVelScale=<0.4 0.4 0.4>
    gravityOverride {z=-980}
    physicsBodyDefinitions [{
      boundBone {boneName=dyn_leg_l}
      boxExtents=<29 29 29>
      constraintSetup {angularLimitsMax=<1 8 11>, angularLimitsMin=<-1 -8 -15>, angularTarget=<1 0 0>}
      localJointOffset=<2.6 0 0>
     }]
    simSpaceSettings {maxAngularAcceleration=50, maxAngularVelocity=5, simSpaceAngularAlpha=0.8}
   }
   n4 = Anim Dynamics - Bone: dyn_leg_r(ComponentPose=@n3, Alpha=1, ExternalForce="0,0,0") {
    angularDampingOverride=1
    angularSpringConstant=20000
    bAngularSpring=true
    bOverrideAngularDamping=true
    bUseGravityOverride=true
    gravityScale=0.35
    simulationSpace=RootRelative
    boundBone {boneName=dyn_leg_r}
    componentAppliedLinearAccClamp=<2500 2500 2500>
    componentLinearAccScale=<0.3 0.3 0.3>
    componentLinearVelScale=<0.3 0.3 0.3>
    gravityOverride {z=-980}
    physicsBodyDefinitions [{
      boundBone {boneName=dyn_leg_r}
      boxExtents=<30 30 30>
      constraintSetup {angularLimitsMax=<0 5 9>, angularLimitsMin=<0 -5 -11>, angularTarget=<1 0 0>}
      localJointOffset=<2 0 0>
     }]
    simSpaceSettings {maxAngularAcceleration=55, maxAngularVelocity=5.5, simSpaceAngularAlpha=0.8}
   }
   n5 = Anim Dynamics - Bone: dyn_head_l(ComponentPose=@n4, Alpha=1, ExternalForce="0,0,0") {
    angularDampingOverride=1
    angularSpringConstant=18000
    bAngularSpring=true
    bOverrideAngularDamping=true
    bUseGravityOverride=true
    gravityScale=0.35
    simulationSpace=RootRelative
    boundBone {boneName=dyn_head_l}
    componentAppliedLinearAccClamp=<3000 3000 3000>
    componentLinearAccScale=<0.3 0.3 0.3>
    componentLinearVelScale=<0.35 0.35 0.35>
    gravityOverride {z=-980}
    physicsBodyDefinitions [{
      boundBone {boneName=dyn_head_l}
      boxExtents=<35 35 35>
      constraintSetup {angularLimitsMax=<6 7.7 12.5>, angularLimitsMin=<-6 -7.7 -8.5>, angularTarget=<1 0 0>}
      localJointOffset=<2 0 0>
     }]
    simSpaceSettings {maxAngularAcceleration=60, maxAngularVelocity=6, simSpaceAngularAlpha=0.8}
   }
   n6 = "Transform (Modify) Bone - Bone: dyn_arm_r"(
    ComponentPose=Local To Component(LocalPose=Use cached pose 'Input saved pose'())
    Alpha=1
    Translation="0,0,0"
    Rotation="0.232,1.25,1"
    Scale="1,1,1"
   ) {rotationMode=BMM_Additive, rotationSpace=BCS_ParentBoneSpace, boneToModify {boneName=dyn_arm_r}, rotation=<rot 0.232 1.25 1>}
   n7 = "Transform (Modify) Bone - Bone: dyn_leg_l"(ComponentPose=@n6, Alpha=1, Translation="0,0,0", Rotation="-0.488,1.473,2.434", Scale="1,1,1") {rotationMode=BMM_Additive, rotationSpace=BCS_ParentBoneSpace, boneToModify {boneName=dyn_leg_l}, rotation=<rot -0.488 1.473 2.434>}
   n8 = "Transform (Modify) Bone - Bone: dyn_leg_r"(ComponentPose=@n7, Alpha=1, Translation="0,0,0", Rotation="-0.044,3.370,11.204", Scale="1,1,1") {rotationMode=BMM_Additive, rotationSpace=BCS_ParentBoneSpace, boneToModify {boneName=dyn_leg_r}, rotation=<rot -0.044 3.37 11.204>}
   n9 = Anim Dynamics - Bone: dyn_arm_r(ComponentPose=@n8, Alpha=1, ExternalForce="0,0,0") {
    angularDampingOverride=1
    angularSpringConstant=18000
    bAngularSpring=true
    bOverrideAngularDamping=true
    bUseGravityOverride=true
    gravityScale=0.35
    simulationSpace=RootRelative
    boundBone {boneName=dyn_arm_r}
    componentAppliedLinearAccClamp=<2500 2500 2500>
    componentLinearAccScale=<0.6 0.6 0.6>
    componentLinearVelScale=<0.6 0.6 0.6>
    gravityOverride {z=-980}
    physicsBodyDefinitions [{
      boundBone {boneName=dyn_arm_r}
      boxExtents=<29 32 28.8>
      constraintSetup {angularLimitsMax=<5 10 20>, angularLimitsMin=<-5 -10 -28>, angularTarget=<1 0 0>}
      localJointOffset=<2.9 0 0>
     }]
    simSpaceSettings {maxAngularAcceleration=60, maxAngularVelocity=5.5, simSpaceAngularAlpha=0.2}
   }
   n10 = Anim Dynamics - Bone: dyn_arm_l(ComponentPose=@n9, Alpha=1, ExternalForce="0,0,0") {
    angularDampingOverride=1
    angularSpringConstant=17000
    bAngularSpring=true
    bOverrideAngularDamping=true
    bUseGravityOverride=true
    gravityScale=0.35
    simulationSpace=RootRelative
    boundBone {boneName=dyn_arm_l}
    componentAppliedLinearAccClamp=<3000 3000 3000>
    componentLinearAccScale=<0.3 0.3 0.3>
    componentLinearVelScale=<0.3 0.3 0.3>
    gravityOverride {z=-880}
    physicsBodyDefinitions [{
      boundBone {boneName=dyn_arm_l}
      boxExtents=<20 30 28>
      constraintSetup {angularLimitsMax=<4 7 20>, angularLimitsMin=<-3 -7 -24>, angularTarget=<1 0 0>}
      localJointOffset=<2.4 0 0>
     }]
    simSpaceSettings {maxAngularAcceleration=65, maxAngularVelocity=6, simSpaceAngularAlpha=0.75}
   }
   n11 = Anim Dynamics - Bone: dyn_leg_l(ComponentPose=@n10, Alpha=1, ExternalForce="0,0,0") {
    angularDampingOverride=1
    angularSpringConstant=17000
    bAngularSpring=true
    bOverrideAngularDamping=true
    bUseGravityOverride=true
    gravityScale=0.35
    simulationSpace=RootRelative
    boundBone {boneName=dyn_leg_l}
    componentAppliedLinearAccClamp=<2500 2500 2500>
    componentLinearAccScale=<0.4 0.4 0.4>
    componentLinearVelScale=<0.4 0.4 0.4>
    gravityOverride {z=-1080}
    physicsBodyDefinitions [{
      boundBone {boneName=dyn_leg_l}
      boxExtents=<29 29 29>
      constraintSetup {angularLimitsMax=<7 8 17>, angularLimitsMin=<-7 -8 -24>, angularTarget=<1 0 0>}
      localJointOffset=<2.5 0 0>
     }]
    simSpaceSettings {maxAngularAcceleration=50, maxAngularVelocity=4.5, simSpaceAngularAlpha=0.6}
   }
   n12 = Anim Dynamics - Bone: dyn_leg_r(ComponentPose=@n11, Alpha=1, ExternalForce="0,0,0") {
    angularDampingOverride=1
    angularSpringConstant=20000
    bAngularSpring=true
    bOverrideAngularDamping=true
    bUseGravityOverride=true
    gravityScale=0.35
    simulationSpace=RootRelative
    boundBone {boneName=dyn_leg_r}
    componentAppliedLinearAccClamp=<2500 2500 2500>
    componentLinearAccScale=<0.3 0.3 0.3>
    componentLinearVelScale=<0.3 0.3 0.3>
    gravityOverride {z=-780}
    physicsBodyDefinitions [{
      boundBone {boneName=dyn_leg_r}
      boxExtents=<30 30 30>
      constraintSetup {angularLimitsMax=<6 5 19>, angularLimitsMin=<-6 -5 -21>, angularTarget=<1 0 0>}
      localJointOffset=<1.9 0 0>
     }]
    simSpaceSettings {maxAngularAcceleration=55, maxAngularVelocity=5, simSpaceAngularAlpha=0.6}
   }
   n13 = Anim Dynamics - Bone: dyn_head_l(ComponentPose=@n12, Alpha=1, ExternalForce="0,0,0") {
    angularDampingOverride=1
    angularSpringConstant=18000
    bAngularSpring=true
    bOverrideAngularDamping=true
    bUseGravityOverride=true
    gravityScale=0.35
    simulationSpace=RootRelative
    boundBone {boneName=dyn_head_l}
    componentAppliedLinearAccClamp=<3000 3000 3000>
    componentLinearAccScale=<0.3 0.3 0.3>
    componentLinearVelScale=<0.35 0.35 0.35>
    gravityOverride {z=-980}
    physicsBodyDefinitions [{
      boundBone {boneName=dyn_head_l}
      boxExtents=<35 35 35>
      constraintSetup {angularLimitsMax=<6 7.7 12.5>, angularLimitsMin=<-6 -7.7 -8.5>, angularTarget=<1 0 0>}
      localJointOffset=<2 0 0>
     }]
    simSpaceSettings {maxAngularAcceleration=60, maxAngularVelocity=6, simSpaceAngularAlpha=0.7}
   }
   Output Pose(Result=Blend Poses by bool(BlendPose_0=Component To Local(ComponentPose=@n13), BlendPose_1=Component To Local(ComponentPose=@n5), bActiveValue=Get bIsInFrontEnd())) {layerGroup=None}
   Save cached pose 'Input saved pose'(Pose=Input Pose()) {cacheName=Input saved pose}
  ]}
 EventGraph {nodes [
   TryGetPawnOwner()
   Event BlueprintUpdateAnimation()
  ]}
]
variables []
